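# Behavior tree that makes the Zeus AI perform its lightning jump, stored as a BT injection in the AIGods Zeus plugin.
# A root Selector runs a single Execute Gameplay Ability task that triggers the ability tagged Ability.AIGods.Zeus.LightningJump.
# It uses the BB_LightningJump blackboard and has no decorators or services, so any gating happens outside this tree.
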BehaviorTree /AIGods_Zeus/BTInjection/BT_LightningJump
blackboard=&/BB_LightningJump
root {
 name=Selector
 class=BTComposite_Selector
 children [GameplayAbility {
   class=FortBTTask_ExecuteGameplayAbility
   decorators []
   properties {nodeName=GameplayAbility, gameplayAbilityTagContainer {defaultValue {gameplayTags [{tagName=Ability.AIGods.Zeus.LightningJump}]}}}
  }]
 decorators []
 properties {
  children [{childComposite=null, childTask=&self:FortBTTask_ExecuteGameplayAbility_0, decorators []}]
  services []
 }
 services []
}
